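# Sound cue for the destruction sound of the ceiling gas trap in the Fort traps audio set.
# A Random node picks one of four equally weighted wave variations of the destroyed sound.
# It uses the shared Fort traps sound class and attenuation at reduced volume, with a concurrency override that blocks new instances once the limit is reached.
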SoundCue /Game/Sounds/Fort_Traps/Ceiling_Gas_Trap/CeilingTrap_Gas_Destroyed_Cue
root=Random_28
nodes [
 {id=Random_28, children [WavePlayer_99, WavePlayer_100, WavePlayer_101, WavePlayer_102], properties {weights [1,1,1,1]}}
 {id=WavePlayer_100, properties {soundWaveAssetPtr=&/CeilingTrap_Gas_Destroyed_003}}
 {id=WavePlayer_101, properties {soundWaveAssetPtr=&/CeilingTrap_Gas_Destroyed_002}}
 {id=WavePlayer_102, properties {soundWaveAssetPtr=&/CeilingTrap_Gas_Destroyed_001}}
 {id=WavePlayer_99, properties {soundWaveAssetPtr=&/CeilingTrap_Gas_Destroyed_004}}
]
properties {
 attenuationSettings=&Fort_Traps_Att
 bOverrideConcurrency=true
 maxDistance=6000
 soundClassObject=&Fort_Traps
 volumeMultiplier=0.5
 concurrencyOverrides {resolutionRule=PreventNew}
}
