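# Sound cue for the ready sound of the ceiling gas trap.
# A random node picks between two equally weighted wave variants, CeilingTrap_Gas_Ready_001 and 002.
# It plays at reduced volume through the Fort_Traps sound class and shared trap attenuation.
# A concurrency override prevents new instances from starting once the limit is reached.
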SoundCue /Game/Sounds/Fort_Traps/Ceiling_Gas_Trap/CeilingTrap_Gas_Ready_Cue
root=Random_26
nodes [
 {id=Random_26, children [WavePlayer_93, WavePlayer_94], properties {weights [1,1]}}
 {id=WavePlayer_93, properties {soundWaveAssetPtr=&/CeilingTrap_Gas_Ready_002}}
 {id=WavePlayer_94, properties {soundWaveAssetPtr=&/CeilingTrap_Gas_Ready_001}}
]
properties {
 attenuationSettings=&Fort_Traps_Att
 bOverrideConcurrency=true
 maxDistance=6000
 soundClassObject=&Fort_Traps
 volumeMultiplier=0.5
 concurrencyOverrides {resolutionRule=PreventNew}
}
